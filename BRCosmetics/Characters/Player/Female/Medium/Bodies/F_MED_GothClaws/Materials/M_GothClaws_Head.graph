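# Head material for the GothClaws female medium player character, built on the shared 2023 character parent functions for skinned, morphing meshes.
# PreFX output passes through a sub-UV texture function, masked by a selectable Tech Art Mask channel, then through an eyelash function reused from ChillCat.
# PostFX finishes the material attributes, with shading model chosen by expression.
# The same sub-UV mask also drives temporal responsiveness.
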Material /BRCosmetics/Characters/Player/Female/Medium/Bodies/F_MED_GothClaws/Materials/M_GothClaws_Head
graph {
 properties {
  bAutomaticallySetUsageInEditor=false
  bUseMaterialAttributes=true
  bUsedWithMeshDeformer=true
  bUsedWithMorphTargets=true
  bUsedWithSkeletalMesh=true
  pixelDepthOffsetMode=PDOM_Legacy
  previewMesh=/BRCosmetics/Characters/Player/Female/Medium/Bodies/F_MED_ChillCat/Meshes/Parts/F_MED_ChillCat_FaceAcc
  shadingModel=MSM_FromMaterialExpression
  usedShadingModels=Default Lit | Subsurface | Clear Coat | Subsurface Profile | Cloth
 }
 nodes [
  n1 = MaterialFunctionCall() {materialFunction=&MF_CharacterParent_2023_PreFX}
  n2 = ChannelMaskParameter(None=@n1.Tech Art Mask) {parameterName=Sub UV Mask Channel}
  MaterialFunctionCall(
   Material=MaterialFunctionCall!(In=MaterialFunctionCall!(Alpha=@n2.R, Material=@n1.Result) {materialFunction=/Game/Weapons/WeaponSkins/MaterialFunction/SubUVTextures}.Result) {materialFunction=/BRCosmetics/Characters/Player/Female/Medium/Bodies/F_MED_ChillCat/Materials/MF_ChillCat_Eyelash}.Result
   M=@n1.M
   Pre Skinned Local Position=@n1.M
   Distance Blend=@n1.M
   Fresnel=@n1.M
  ) {materialFunction=&MF_CharacterParent_2023_PostFX}
  TemporalResponsivenessOutput!(None=@n2.R)
 ]
}
