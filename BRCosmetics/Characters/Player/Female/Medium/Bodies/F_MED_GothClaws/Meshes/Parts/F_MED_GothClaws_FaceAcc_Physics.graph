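# Physics asset for the face accessory part of the female medium GothClaws player cosmetic, covering a medallion on the upper spine.
# It has three capsule bodies: a large one on spine_05, a tiny one on base_medallion and a small one on dyn_medallion_01.
# The base-to-spine constraint leaves all angular axes free.
# The dynamic medallion joins the base with twist and swing 2 locked, so it swings hinge-like on one axis.
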PhysicsAsset /BRCosmetics/Characters/Player/Female/Medium/Bodies/F_MED_GothClaws/Meshes/Parts/F_MED_GothClaws_FaceAcc_Physics
bodies {
 base_medallion [{shapeType=Capsule, length=0.505, radius=0.505, shapeName=base_medallion_capsule}]
 dyn_medallion_01 [{shapeType=Capsule, length=3, radius=1.5, shapeName=dyn_medallion_01_capsule, center=<2.2769 0 1>, rotation=<rot 0 89.236 -88.9075>}]
 spine_05 [{shapeType=Capsule, length=18, radius=18, shapeName=spine_05_capsule, center=<-4.3626 -7.483 0>, rotation=<rot 90 147.592 0>}]
}
constraints [
 {
  bone1Name=base_medallion
  bone2Name=spine_05
  swing1LimitDegrees=45
  swing1Motion=Free
  swing2LimitDegrees=45
  swing2Motion=Free
  twistLimitDegrees=45
  twistMotion=Free
 }
 {
  bone1Name=dyn_medallion_01
  bone2Name=base_medallion
  swing1LimitDegrees=40
  swing2LimitDegrees=5
  swing2Motion=Locked
  twistLimitDegrees=45
  twistMotion=Locked
 }
]
