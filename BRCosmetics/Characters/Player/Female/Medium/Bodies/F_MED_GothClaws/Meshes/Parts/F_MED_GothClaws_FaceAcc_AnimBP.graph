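# Animation Blueprint for the GothClaws female outfit's face accessory mesh, adding secondary motion to its ear and cap ring bones.
# It copies the attached parent's pose into a Control Rig fed a Jump value, which Event Graph sets from jumping or falling.
# RigidBody physics relative to spine_05 follows, then a constraint on the right ear bone and Anim Dynamics springs on two cap ring bones.
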AnimBlueprint /BRCosmetics/Characters/Player/Female/Medium/Bodies/F_MED_GothClaws/Meshes/Parts/F_MED_GothClaws_FaceAcc_AnimBP
target_skeleton=&/F_MED_GothClaws_FaceAcc_Skeleton
graphs [
 AnimGraph {nodes [
   n1 = Control Rig!(Source=Copy Pose From Mesh() {bUseAttachedParent=true}, Alpha, Jump:double=Get Jump() -> (Jump:double)) -> (Pose) {controlRigAssetReference {blueprintRigClass=&/F_MED_GothClaws_FaceAcc_CtrlRig1.F_MED_GothClaws_FaceAcc_CtrlRig1_C}}
   n2 = RigidBody(ComponentPose=Local To Component(LocalPose=@n1)) {
    bClampLinearTranslationLimitToRefPose=true
    bOverrideWorldGravity=true
    overridePhysicsAsset=&/F_MED_GothClaws_FaceAcc_Physics
    simulationSpace=BaseBoneSpace
    alphaBoolBlend {blendOption=Linear}
    baseBoneRef {boneName=spine_05}
    componentAppliedLinearAccClamp=<0 0 0>
    overrideWorldGravity {z=-980}
    simSpaceSettings {
     dampingAlpha=0
     maxAngularAcceleration=10
     maxAngularVelocity=8
     maxLinearAcceleration=1000
     maxLinearVelocity=500
     worldAlpha=1
     externalLinearDragV=<0.1 0.1>
    }
   }
   n3 = "Constraint - dyn_ear_02_r (PPP)"!(ComponentPose=@n2, Alpha=1, ConstraintWeights_0=0.3696, ConstraintWeights_1:float=0.3685, ConstraintWeights_2:float=0.2619) -> (Pose) {
    boneToModify {boneName=dyn_ear_02_r}
    constraintSetup [{targetBone {boneName=head}}, {targetBone {boneName=dyn_ear_02_r}}, {targetBone {boneName=dyn_ear_01_r}}]
    constraintWeights [1,1,1]
   }
   n4 = Anim Dynamics - Bone: dyn_cap_ring_b(ComponentPose=@n3, Alpha=1, ExternalForce="0,0,0") {
    angularDampingOverride=0.9999
    angularSpringConstant=10000
    bAngularSpring=true
    bOverrideAngularDamping=true
    bUseGravityOverride=true
    gravityScale=0.35
    simulationSpace=RootRelative
    boundBone {boneName=dyn_cap_ring_b}
    componentAppliedLinearAccClamp=<3000 3000 3000>
    componentLinearAccScale=<0.9 0.9 0.9>
    componentLinearVelScale=<0.9 0.9 0.9>
    gravityOverride {z=-980}
    physicsBodyDefinitions [{
      boundBone {boneName=dyn_cap_ring_b}
      boxExtents=<20 30 30>
      constraintSetup {angularLimitsMax=<0 15 10>, angularLimitsMin=<0 -15 -10>, angularTarget=<1 0 0>}
      localJointOffset=<4 0 0>
     }]
   }
   n5 = Anim Dynamics - Bone: dyn_cap_ring_a(ComponentPose=@n4, Alpha=1, ExternalForce="0,0,0") {
    angularDampingOverride=0.9999
    angularSpringConstant=10000
    bAngularSpring=true
    bOverrideAngularDamping=true
    bUseGravityOverride=true
    gravityScale=0.35
    simulationSpace=RootRelative
    boundBone {boneName=dyn_cap_ring_a}
    componentAppliedLinearAccClamp=<3000 3000 3000>
    componentLinearAccScale=<1 1 1>
    componentLinearVelScale=<1 1 1>
    gravityOverride {z=-980}
    physicsBodyDefinitions [{
      boundBone {boneName=dyn_cap_ring_a}
      boxExtents=<20 30 30>
      constraintSetup {angularLimitsMax=<0 15 10>, angularLimitsMin=<0 -15 -20>, angularTarget=<1 0 0>}
      localJointOffset=<4 0 0>
     }]
   }
   Output Pose(Result=Component To Local(ComponentPose=@n5)) {layerGroup=None}
  ]}
 EventGraph {nodes [
   Event BlueprintUpdateAnimation()
   n1 = Get bIsJumping()
   n2 = Get bIsFalling()
   > Set Jump(Jump:double="To Float (Boolean)"(InBool=OR Boolean(A=@n1, B=@n2))) -> (Output_Get:double)
  ]}
]
variables [Jump {type=number}]
